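; Snow Doofy buildup actor that shows a staged smoke pillar, funnel top and storm screen effect.
; SetEffectsState stores the stage, hides the actor at zero or below, swaps smoke particles, and snaps meshes when not at runtime.
; At runtime it waits for pedestal animation, fires a local quake cue, then timelines crossfade or enlarge smoke by stage.
; Construction applies the stage, configures the storm's dynamic material, and scales its bounds to fit.
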(event Custom|SetEffectsState
  (params (Stage :type Integer) (Runtime :type Boolean))
  (bind _finalstage (Variables|Default|GetFinalStage))
  (bind _buildupstage (Variables|Default|GetBuildupStage))
  (bind _returnvalue (>= _buildupstage _finalstage))
  (bind _buildupstage_2 (Variables|Default|GetBuildupStage))
  (bind _returnvalue_2 (<= _buildupstage_2 4))
  (if Runtime
    (if (== Stage (Variables|Default|GetBuildupStage))
      _
      (else
        _)))
  (Variables|Default|SetBuildupStage Stage)
  (Rendering|SetActorHiddenInGame (<= Stage 0))
  (Effects|Components|ParticleSystem|SetTemplate (Variables|Default|GetP_Snow_Doofy_Wispy_Smoke) (select (>= (Variables|Default|GetBuildupStage) (Variables|Default|GetFinalStage)) "/Figment_S08/Doofy_Buildup_8_5_1_Migrated/Athena/Prototype/Blueprints/White/FX/P_Snow_Doofy_Wispy_Smoke_4.P_Snow_Doofy_Wispy_Smoke_4" "/Figment_S08/Doofy_Buildup_8_5_1_Migrated/Athena/Prototype/Blueprints/White/FX/P_Snow_Doofy_Wispy_Smoke_2.P_Snow_Doofy_Wispy_Smoke_2"))
  (if Runtime
    (Utilities|FlowControl|Delay 14.0) ; Give the gateway pedestal animation time to finish
    ;[1 1 1] Update during gameplay
    (GameplayCue|ExecuteGameplayCue%28Burst%29Local (Pawn|GetControlledPawn (Utilities|Array|Get(acopy) (Fortnite|GetLocalFortPlayerControllers true true))) (Variables|Default|GetBuildupQuakeCue) (Ability|GameplayCue|MakeGameplayCueParameters :NormalizedMagnitude 0.0 :RawMagnitude 0.0 :Location "0,0,0" :Normal "0,0,0" :EffectCauser self :GameplayEffectLevel 1 :AbilityLevel 1 :bReplicateLocationWhenUsingMinimalRepProxy false)) ; Gameplay cue with audio and cam shake
    (if (== (Variables|Default|GetBuildupStage) (Variables|Default|GetFinalStage))
      (Rendering|SetVisibility (Variables|Default|GetSmokeFunnelTop) true false)
      (bind (_direction _lerp) (Timeline|StateTransition PlayFromStart)
        (:Update
          (Rendering|Material|SetScalarParameterValueonMaterials (Variables|Default|GetSM_Snow_Doofy_WispySmoke) "Opacity" (Math|Float|Lerp 1.0 0.0 _lerp))
          (Rendering|Material|SetScalarParameterValueonMaterials (Variables|Default|GetSmokeFunnelTop) "Opacity" (Math|Float|Lerp 0.0 1.0 _lerp))
          (Rendering|Material|SetScalarParameterValueonMaterials (Variables|Default|GetSmokePillarBase) "MinOpacity" (Math|Float|Lerp -1.0 0.7 _lerp)))
        (:Finished
          (Rendering|SetVisibility (Variables|Default|GetSM_Snow_Doofy_WispySmoke) false false)))
      (elif (> (Variables|Default|GetBuildupStage) (Variables|Default|GetFinalStage))
        (bind (_direction_2 _lerp_2) (Timeline|Timeline PlayFromStart)
          (:Update
            (Transformation|SetRelativeScale3D (Variables|Default|GetSmokeFunnelTop) (Math|Vector|Lerp(Vector) "2.000000,2.000000,2.000000" "3.000000,3.000000,3.000000" _lerp_2))
            (Transformation|SetRelativeScale3D (Variables|Default|GetSmokePillarBase) (Math|Vector|Lerp(Vector) "1.500000,1.500000,8.000000" "3.000000,3.000000,8.000000" _lerp_2))))))
    (else
      (Rendering|SetVisibility (Variables|Default|GetSM_Snow_Doofy_WispySmoke) (not _returnvalue) false)
      (Rendering|SetVisibility (Variables|Default|GetSmokeFunnelTop) _returnvalue false)
      (Transformation|SetRelativeScale3D (Variables|Default|GetSmokeFunnelTop) (Math|Vector|SelectVector "2.000000,2.000000,2.000000" "3.000000,3.000000,3.000000" _returnvalue_2))
      (Transformation|SetRelativeScale3D (Variables|Default|GetSmokePillarBase) (Math|Vector|SelectVector "1.500000,1.500000,8.000000" "3.000000,3.000000,8.000000" _returnvalue_2)))))

(fn ConstructionScript
  (bind (_min _max) (Components|StaticMesh|GetLocalBounds (Variables|Default|GetSM_Storm_ScreenEffect)))
  (CallFunction|SetEffectsState (Variables|Default|GetBuildupStage))
  (bind _returnvalue (Rendering|Material|CreateDynamicMaterialInstance (Variables|Default|GetSM_Storm_ScreenEffect) 0 0 "None" :!class /Script/Engine.PrimitiveComponent))
  (Rendering|Material|SetScalarParameterValue _returnvalue "F_RadiusTop" (Variables|Default|GetRadiusTop) :!class /Script/Engine.MaterialInstanceDynamic)
  (Rendering|Material|SetScalarParameterValue _returnvalue "F_BlendRadius" (Variables|Default|GetBlendDistance) :!class /Script/Engine.MaterialInstanceDynamic)
  (Rendering|Material|SetScalarParameterValue _returnvalue "F_RadiusBottom" (Variables|Default|GetRadiusBottom) :!class /Script/Engine.MaterialInstanceDynamic)
  (Rendering|Material|SetScalarParameterValue _returnvalue "F_Height" (Variables|Default|GetHeight) :!class /Script/Engine.MaterialInstanceDynamic)
  (Rendering|Material|SetScalarParameterValue _returnvalue "F_BlendTop" (Variables|Default|GetBlendDistance_Top) :!class /Script/Engine.MaterialInstanceDynamic)
  (Rendering|Material|SetScalarParameterValue _returnvalue "F_BlendBottom" (Variables|Default|GetBlendDistance_Bottom) :!class /Script/Engine.MaterialInstanceDynamic)
  (Rendering|SetBoundsScale (Variables|Default|GetSM_Storm_ScreenEffect) (/ (Math|Float|Max(Float) (Math|Float|float+float (Variables|Default|GetHeight) (Variables|Default|GetBlendDistance_Top) (+ (Variables|Default|GetRadiusTop) (Variables|Default|GetBlendDistance)) (Variables|Default|GetBlendDistance)) (Math|Float|float+float (Variables|Default|GetHeight) (Variables|Default|GetBlendDistance_Bottom) (+ (Variables|Default|GetRadiusBottom) (Variables|Default|GetBlendDistance)) (Variables|Default|GetBlendDistance))) (Math|Float|Max(Float) (.x _max) (.y _max) (.z _max)))))
