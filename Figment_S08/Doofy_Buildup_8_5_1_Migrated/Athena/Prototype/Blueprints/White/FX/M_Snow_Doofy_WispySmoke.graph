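# Translucent, unlit, two-sided material for wispy smoke on a mesh column, with a glowing base.
# World position offset widens it radially toward the top and adds sine drift, using vertex red for variation and green for drift and opacity masking.
# Opacity blends panning trail and storm-cloud textures by height, with angle and near-camera fades.
# Matching normals fake sun highlights over time-of-day shadow tints, with a cheaper low-quality and mobile fallback.
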Material /Figment_S08/Doofy_Buildup_8_5_1_Migrated/Athena/Prototype/Blueprints/White/FX/M_Snow_Doofy_WispySmoke
graph {
 properties {blendMode=BLEND_Translucent, pixelDepthOffsetMode=PDOM_Legacy, shadingModel=MSM_Unlit, twoSided=true, usedShadingModels=Unlit}
 nodes [
  n1 = ScalarParameter() {defaultValue=5, parameterName=WPOWiden}
  n2 = ComponentMask(None=TextureCoordinate()) {g=true}
  n3 = OneMinus(None=@n2)
  n4 = Power(Base=@n3) {constExponent=1.5}
  n5 = Multiply(A=ScalarParameter() {defaultValue=5, parameterName=WPOTopWidthMult}, B=@n4)
  n6 = Time()
  n7 = ScalarParameter() {defaultValue=-0.02, parameterName=WPOWidenSpeed}
  n8 = ComponentMask(None=Multiply(A=TextureCoordinate(), B=ScalarParameter() {defaultValue=1.5, parameterName=WPOTile})) {g=true}
  n9 = VertexColor()
  n10 = Add(A=Add(A=Multiply(A=@n6, B=@n7), B=@n8), B=Multiply(A=@n9.R) {constB=-0.4})
  n11 = Add(A=@n3, B=ConstantBiasScale(None=Sine(None=@n10) {period=0.51})) {constB=0}
  n12 = Multiply(A=Subtract(A=WorldPosition().XYZ, B=ObjectPositionWS()), B=Constant3Vector() {constant=<rgba 1 1 0 0>})
  n13 = ScalarParameter() {defaultValue=0.02, parameterName=WPODriftSpeed}
  n14 = Add(A=Add(A=Multiply(A=@n13, B=@n6), B=@n8), B=Multiply(A=@n9.R) {constB=0.1})
  n15 = Sine(None=@n14) {period=0.62}
  n16 = Sine(None=@n14) {period=0.57}
  n17 = Sine(None=@n14) {period=0.49}
  n18 = Multiply(
   A=LinearInterpolate(Alpha=@n2) {constA=1, constB=0.18}
   B=AppendVector(A=AppendVector(A=@n15, B=@n16), B=Multiply(A=@n17) {constA=0.2, constB=0.15})
  )
  Add(
   A=Multiply(A=Add(A=@n1, B=@n5), B=Multiply(A=@n11, B=@n12))
   B=Multiply(A=Multiply(A=@n18, B=@n9.G), B=ScalarParameter() {defaultValue=1000, parameterName=WPODrift})
  )
  n19 = TextureSample(UVs=Add(A=TextureCoordinate() {vTiling=0}, B=Constant2Vector() {g=0.5})) {samplerType=SAMPLERTYPE_Normal, texture=/Game/Effects/Fort_Effects/Textures/Mask/T_Sphere_N}
  AppendVector(A=ComponentMask(None=@n19.RGB) {g=true, r=true}, B=Multiply(A=@n19.B, B=TwoSidedSign()))
  n20 = VertexColor()
  n21 = ScalarParameter() {defaultValue=2.5, parameterName=Tile}
  n22 = VertexColor()
  n23 = LinearInterpolate(Alpha=@n22.R) {constA=0.75}
  n24 = Multiply(A=TextureCoordinate() {coordinateIndex=1}, B=AppendVector(A=Multiply(A=@n21, B=@n23), B=Constant() {r=1}))
  n25 = Multiply(A=Multiply(A=Time(), B=LinearInterpolate(Alpha=@n22.R) {constA=0.35}), B=ScalarParameter() {defaultValue=0.6, parameterName=PanSpeed})
  n26 = Panner(Coordinate=@n24, Time=@n25) {speedX=-0.1}
  n27 = Multiply(
   A=Panner(Coordinate=@n24, Time=@n25) {speedX=-0.05, speedY=0.02}
   B=ComponentMask(None=VectorParameter() {parameterName=TopTileMult, defaultValue=<rgba 1 0.5 0 1>}.RGB) {g=true, r=true}
  ) {constB=0.7}
  n28 = ComponentMask(None=TextureCoordinate()) {r=true}
  n29 = Constant() {r=0.5}
  n30 = LinearInterpolate(
   A=TextureSample(UVs=@n26) {samplerType=SAMPLERTYPE_Grayscale, texture=&T_Smoke_Trail_Soft}.G
   B=Multiply(
    A=TextureSample(UVs=@n27) {samplerType=SAMPLERTYPE_LinearGrayscale, texture=/Game/VisualThreatManager/StormVisuals/LargeClouds/T_StormCloudDetail01_M}.G
    B=SphereMask(A=@n28, B=@n29, Radius=@n29) {hardnessPercent=70}
   )
   Alpha=Multiply(A=@n3, B=@n3)
  )
  n31 = DotProduct(A=CameraVectorWS(), B=VertexNormalWS())
  n32 = ScalarParameter() {defaultValue=0.8, parameterName=AngleFadeBase}
  n33 = ScalarParameter() {defaultValue=2, parameterName=AngleFadeTop}
  n34 = Clamp(None=Power(Base=Abs(None=@n31), Exp=LinearInterpolate(A=@n32, B=@n33, Alpha=@n4)))
  n35 = MaterialFunctionCall(Distance=ScalarParameter() {defaultValue=5000, parameterName=NearFadeDistance}) {materialFunction=&Near_Camera_Fade}
  Multiply(A=Multiply(A=Multiply(A=@n30, B=@n20.G), B=Multiply(A=@n34, B=@n35.Result)), B=ScalarParameter() {defaultValue=1, parameterName=Opacity})
  n36 = ComponentMask(None=TextureCoordinate()) {g=true}
  n37 = ScalarParameter() {defaultValue=20, parameterName=GlowBasePower}
  n38 = Multiply(A=VectorParameter() {parameterName=GlowBase, defaultValue=<rgba 2 0.3347 0 0>}.RGB, B=Saturate(None=Power(Base=@n36, Exp=@n37)))
  n39 = ScalarParameter() {defaultValue=1, parameterName=BaseGlow}
  n40 = CollectionParameter() {
   collection=&FortniteMaterialParameters
   parameterName=SunAndMoonModelDirectionalVector
  }
  n41 = Normalize(VectorInput=ComponentMask(None=@n40) {b=true, g=true, r=true})
  n42 = Transform(None=LinearInterpolate(
    A=TextureSample(UVs=@n26) {samplerType=SAMPLERTYPE_Normal, texture=/Game/Effects/Fort_Effects/Textures/Smoke/T_Smoke_Trail_Soft_N}.RGB
    B=TextureSample(UVs=@n27) {samplerType=SAMPLERTYPE_Normal, texture=/Game/VisualThreatManager/StormVisuals/LargeClouds/T_StormCloudDetail01_N}.RGB
    Alpha=@n3
   ))
  n43 = Power(Base=Clamp(None=DotProduct(A=@n41, B=@n42)), Exp=ScalarParameter() {defaultValue=2, parameterName=HighlightPower})
  n44 = VectorParameter() {group=Colors, parameterName=NightTint, defaultValue=<rgba 0.094 0.1 0.0927 1>}
  n45 = Multiply(
   A=VectorParameter() {group=Colors, parameterName=HighlightColor, defaultValue=<rgba 0.3724 0.4956 0.5 0>}.RGB
   B=MaterialFunctionCall(
    Color_Morning=Constant() {r=0.5}
    Color_Day=Constant() {r=0.7}
    Color_Evening=VectorParameter() {group=Colors, parameterName=EveningTint, defaultValue=<rgba 0.99 0.6691 1 1>}.RGB
    Color_Night=@n44.RGB
   ) {materialFunction=&Color_Time_of_Day}.Result
  )
  n46 = VectorParameter() {group=Colors, parameterName=ShadowNight, defaultValue=<rgba 0.25 0.2445 0.2279 0>}
  n47 = MaterialFunctionCall(
   Color_Morning=VectorParameter() {group=Colors, parameterName=ShadowMorning, defaultValue=<rgba 0.5 0.3633 0.349 0>}.RGB
   Color_Day=VectorParameter() {group=Colors, parameterName=ShadowDay, defaultValue=<rgba 0.4671 0.4531 0.6 0>}.RGB
   Color_Evening=VectorParameter() {group=Colors, parameterName=ShadowEvening, defaultValue=<rgba 0.3703 0.3458 0.4 1>}.RGB
   Color_Night=@n46.RGB
  ) {materialFunction=&Color_Time_of_Day}
  n48 = ScalarParameter() {defaultValue=0.4, parameterName=ShadowMult}
  n49 = LinearInterpolate(A=@n46.RGB, B=@n44.RGB, Alpha=@n30)
  n50 = QualitySwitch(Default=Add(A=Multiply(A=@n43, B=@n45) {constB=0}, B=Multiply(A=@n47.Result, B=@n48)), Low=@n49)
  n51 = FeatureLevelSwitch(Default=@n50, ES2=@n49, ES3_1=@n49, SM4_REMOVED=@n50)
  Multiply(A=Add(A=Multiply(A=@n38, B=@n39), B=Multiply(A=@n51) {constB=0.25}), B=LinearInterpolate(Alpha=@n30) {constA=0.4})
 ]
}
